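; Fortnite controller effect in the shared NPC conversation services that applies a configurable gameplay effect to the requester, with magnitude read from a health cost curve.
; It takes the ability system component from the effect request context and builds a spec from the configured effect class.
; It evaluates the curve table row, assigns the result to a SetByCaller cost tag, and applies the spec to that component only if the row is found.
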(fn ApplyEffectInternal
  (params (RequestContext :type Effect%20Request%20Context%20Structure%20%28by%20ref%29))
  (bind (_curvetable _rowname) (Utilities|Struct|BreakCurveTableRowHandle (Variables|Default|GethealthCostCurveTable)))
  (bind _output_get (LocalVariables|SetplayerAbilitySystem (Ability|GetAbilitySystemComponent (Utilities|Struct|BreakEffectRequestContext RequestContext))))
  (bind _returnvalue (Spec|MakeSpecHandle(ByClass) (Variables|Default|GetgameplayEffectToApply)))
  (bind _outxy (DataTable|EvaluateCurveTableRow _curvetable _rowname)
    (:RowFound
      (bind _returnvalue_2 (Ability|GameplayEffect|AssignTagSetbyCallerMagnitude _returnvalue (Variables|Default|GetsetByCallerCostTag) _outxy))
      (GameplayEffects|ApplyGameplayEffectSpecToSelf _output_get _returnvalue_2)
      (return))))
